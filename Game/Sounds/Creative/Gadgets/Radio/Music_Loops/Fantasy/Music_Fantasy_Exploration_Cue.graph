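# Sound cue for the Fantasy exploration music loop, filed under the Creative radio gadget's music.
# A mixer starts a one-shot intro wave alongside a looping wave that is delayed by a fixed fraction of a second.
# An enveloper gives the loop a very short volume ramp-in after the delay.
# Output routes to the base music sound class with overridden, non-spatialized attenuation over a long falloff.
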SoundCue /Game/Sounds/Creative/Gadgets/Radio/Music_Loops/Fantasy/Music_Fantasy_Exploration_Cue
root=Mixer_0
nodes [
 {id=Delay_0, children [Enveloper_0], properties {delayMax=0.285, delayMin=0.285}}
 {
  id=Enveloper_0
  children [WavePlayer_2]
  properties {
   pitchCurve {editorCurveData {keys [{time=0}]}}
   volumeCurve {editorCurveData {keys [<at 0 0.001>,<at 0.0132 0.1784>,<at 0.0207 0.563>,<at 0.025 1>]}}
  }
 }
 {id=Mixer_0, children [WavePlayer_0,Delay_0], properties {inputVolume [1,1]}}
 {id=WavePlayer_0, properties {soundWaveAssetPtr=&/FN_Creative_Fantasy_Explore_Intro}}
 {
  id=WavePlayer_2
  properties {bLooping=true, soundWaveAssetPtr=&/FN_Creative_Fantasy_Explore_Loop}
 }
]
properties {
 bOverrideAttenuation=true
 maxDistance=100000
 soundClassObject=/Game/Sounds/FortSoundClassesAndModes/Fort_11_Base_Music.Fort_11_Base_music
 volumeMultiplier=0.8
 attenuationOverrides {bSpatialize=false, falloffDistance=40000, reverbWetLevelMax=0.2, reverbWetLevelMin=0.1, attenuationShapeExtents {x=60000}}
}
