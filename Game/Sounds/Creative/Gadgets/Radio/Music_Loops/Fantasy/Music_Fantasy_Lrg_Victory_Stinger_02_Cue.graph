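# Sound cue for a large Fantasy victory music stinger (variant 02) in the Creative radio gadget's music loop set.
# A single wave player feeds the output directly, with no looping node or randomization, routed through the SFX sound class at slightly reduced volume.
# Overridden attenuation makes it non-spatialized with a very wide radius and long falloff, plus a light reverb send.
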SoundCue /Game/Sounds/Creative/Gadgets/Radio/Music_Loops/Fantasy/Music_Fantasy_Lrg_Victory_Stinger_02_Cue
root=WavePlayer_2
nodes [{
  id=WavePlayer_2
  properties {soundWaveAssetPtr=&/FN_Creative_Fantasy_Lrg_Victory_Stinger_02}
 }]
properties {
 bOverrideAttenuation=true
 maxDistance=100000
 soundClassObject=&SFX
 volumeMultiplier=0.8
 attenuationOverrides {bSpatialize=false, falloffDistance=40000, reverbWetLevelMax=0.2, reverbWetLevelMin=0.1, attenuationShapeExtents {x=60000}}
}
